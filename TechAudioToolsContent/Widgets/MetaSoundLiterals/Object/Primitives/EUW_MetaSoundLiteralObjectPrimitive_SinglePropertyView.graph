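# Editor utility widget in TechAudioTools that provides a single-property editor for a MetaSound Object literal input.
# A SinglePropertyView targets the widget itself on PreConstruct, displaying its own Object variable with a blank name override.
# Edits broadcast a FieldNotify change for Object.
# The widget reports Input Name as its required viewmodel, then looks up that key in the supplied map and assigns the result as its MetaSoundInputEditorViewModel.
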EditorUtilityWidgetBlueprint /TechAudioToolsContent/Widgets/MetaSoundLiterals/Object/Primitives/EUW_MetaSoundLiteralObjectPrimitive_SinglePropertyView
graphs [
 EventGraph {nodes [
   "On Property Changed (SinglePropertyView)"!() -> (OutputDelegate, then, PropertyName:Name=None) {is=title K2Node_ComponentBoundEvent}
   > Broadcast Field Value Changed!(execute, self:Widget Object Reference, FieldId:Field Notification Id Structure=(FieldName="Object")) -> (then)
   n1 = Event Set Input Viewmodels!() -> (OutputDelegate, then, InputViewModels:Map of Names to Meta Sound Input Viewmodel Object References) {is=title K2Node_Event}
   > n2 = Cast To MetaSoundInputEditorViewModel!(
    execute
    Object=Find(
     TargetMap:Map of Names to Meta Sound Input Viewmodel Object References=@n1.InputViewModels
     Key:"Name (by ref)"=Get Input Name() -> (Input Name:Name=None)
    ) -> (Value:Meta Sound Input Viewmodel Object Reference)
   ) -> (then, CastFailed, AsMetaSound Input Editor Viewmodel:Meta Sound Input Editor Viewmodel Object Reference) {is=title K2Node_DynamicCast}
   > SetMetaSoundInputEditorViewModel!(execute, self, Viewmodel:Meta Sound Input Editor Viewmodel Object Reference=@n2) -> (then)
   Event PreConstruct()
   > SetObject(self=Get SinglePropertyView() -> (SinglePropertyView:Single Property View Object Reference), NewObject=Self-Reference())
  ]}
 GetInputViewModelNames {nodes [
   Get Input Viewmodel Names() {is=title K2Node_FunctionEntry}
   > Return Node!(execute, ReturnValue:Array of Names=Make Array("[0]":Name=Get Input Name() -> (Input Name=None)) -> (Array:Array of Names))
  ]}
]
variables [
 Input Name {type=string, default=None}
 Object {type=Object, default=null}
]
widgets {
 info {rootWidgetClass=&VerticalBox, widgetCount=2}
 widgets [
  VerticalBox_0 {widgetClassPath=&VerticalBox, widget=VerticalBox_0}
  SinglePropertyView {
   widgetClassPath=&SinglePropertyView
   parent=VerticalBox_0
   widget=SinglePropertyView
   properties {nameOverride=" ", propertyName=Object}
   slot {class=VerticalBoxSlot}
  }
 ]
}
